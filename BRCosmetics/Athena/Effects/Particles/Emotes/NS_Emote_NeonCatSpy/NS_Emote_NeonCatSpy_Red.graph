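# Niagara system for the red variant of the NeonCatSpy emote effect, a short burst of red glow, sparks and smoke.
# Flare, rising stretched glow and expanding mesh cone emitters flash and fade, each with light renderers.
# Sprites spawn on sampled character skeleton bones and animate through dynamic material parameters.
# Bright sparks shoot upward against drag, while SubUV smoke drifts with curl noise, darkening from red as it fades.
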NiagaraSystem /BRCosmetics/Athena/Effects/Particles/Emotes/NS_Emote_NeonCatSpy/NS_Emote_NeonCatSpy_Red
dependencies {
 usedDataInterfaces [
  &NiagaraDataInterfaceCurve
  &NiagaraDataInterfaceSkeletalMesh
  &NiagaraDataInterfaceColorCurve
  &NiagaraDataInterfaceVector2DCurve
  &NiagaraDataInterfaceVectorCurve
 ]
 usedDynamicInputs []
 usedModules [&SystemState]
}
emitters {
 FLare001 {
  emitterUpdateScript [
   &EmitterState(Life Cycle Mode=::NewEnumerator1, Loop Behavior=::NewEnumerator1, Loop Duration=1, UseLoopDelay=false)
   &SpawnBurst_Instantaneous(Spawn Count=1, Use Spawn Probability=false)
  ]
  particleSpawnScript [/Niagara/Modules/Spawn/Initialization/InitializeParticle(
    Color=<rgba 1 0.0292 0.0292 1>
    Lifetime=0.25
    Module.Write Color=true
    Module.Write Lifetime=true
    Module.Write Mass=true
    Module.Write Material Random=true
    Module.Write Position=false
    Module.Write Scale=false
    Module.Write SpriteRotation=false
    Module.Write SpriteSize=true
    Position=&SimulationPosition()
    Sprite Size=<400 400>
   )]
  particleUpdateScript [
   &ParticleState()
   &ScaleColor(
    Scale Alpha=&FloatFromCurve(FloatCurve=DataInterface() {propertyValues {Curve {keys [
         {arriveTangent=-0.551, interpMode=RCIM_Cubic, leaveTangent=-0.551, tangentMode=RCTM_User, time=0, value=1}
         {arriveTangent=-2.3347, interpMode=RCIM_Cubic, leaveTangent=-2.3347, tangentMode=RCTM_User, time=1.0011, value=0.002}
        ]}}})
    ScaleA=true
    ScaleRGB=true
   )
   &ScaleSpriteSize(Scale Factor=&Vector2DFromFloat(Value=&FloatFromCurve(FloatCurve=DataInterface() {propertyValues {Curve {keys [<at 0.0239 0>,<at 0.1 1>,<at 1.0011 0.002>]}}}, Scale Curve=2)))
  ]
  rendererClasses [&NiagaraSpriteRendererProperties, &NiagaraLightRendererProperties]
  renderers [
   {rendererClass=&NiagaraSpriteRendererProperties}
   {rendererClass=&NiagaraLightRendererProperties, rendererIndex=1}
  ]
 }
 FN_Emotes_Glow001 {
  emitterUpdateScript [
   &EmitterScalability(
    Module.bApplySpawnCountScale=false
    Module.bApplySpawnCountScaleByDistanceFraction=false
    Module.bUseMaxDistance=false
    Module.bUseMinDistance=false
   ) {enabled=false}
   &EmitterLifeCycle(CompleteOnInactive=true, MaxLoopCount=1, NextLoopDuration=10) {enabled=false}
   &EmitterState()
   &SpawnRate(SpawnRate=20, Use Spawn Probability=false) {enabled=false}
   &SpawnBurst_Instantaneous(Spawn Count=1, Spawn Time=0.05, Use Spawn Probability=false)
  ]
  particleSpawnScript [&InitializeParticle(
    Lifetime=0.2
    Module.Write Color=true
    Module.Write Lifetime=true
    Module.Write Mass=true
    Module.Write Material Random=true
    Module.Write Position=true
    Module.Write Scale=false
    Module.Write SpriteRotation=false
    Module.Write SpriteSize=true
    Position=&SimulationPosition()
    Sprite Size=&Vector2DFromFloat(Value=100)
   )]
  particleUpdateScript [
   &ParticleState()
   &ScaleColor(
    Scale Alpha=&FloatFromCurve(FloatCurve=DataInterface() {propertyValues {Curve {keys [<at 0 0.5>,<at 1.0011 0.002>]}}})
    Scale RGB=&MakeVectorFromLinearColorRGB(Color=&ColorFromCurve(ColorCurve=DataInterface() {propertyValues {
        AlphaCurve {keys [<at -0.0023 1 RCIM_Cubic>,<at 0.9977 1 RCIM_Cubic>]}
        BlueCurve {keys [<at -0.0415 0.3851>,<at 0.1731 0>,<at 1.0885 0.0351>]}
        GreenCurve {keys [<at -0.0415 0.7828>,<at 0.1731 0.05>,<at 1.0885 0.0351>]}
        RedCurve {keys [<at -0.0415 1>,<at 0.1731 1>,<at 1.0885 0.963>]}
       }}))
    ScaleA=true
    ScaleRGB=true
   )
   &ScaleColor(
    Scale RGB=&VectorFromFloat(Value=&FloatFromCurve(FloatCurve=DataInterface() {propertyValues {Curve {keys [<at 0 1>,<at 0.6 2>,<at 1 0>]}}}))
    ScaleA=true
    ScaleRGB=true
   ) {module=ScaleColor001}
   &ScaleSpriteSize(Scale Factor=&Vector2DFromCurve(
     Vector2Curve=DataInterface() {propertyValues {XCurve {keys [<at -0.0011 1>,<at 0.9875 0.4708>]}, YCurve {keys [<at -0.0011 1>,<at 0.9794 2.2959>]}}}
     ScaleCurve=<1 1>
    ))
   &SystemLocation(
    MaskSpawnGroup=false
    Offset=&Multiply_Vector(
     A=&VectorFromCurve(VectorCurve=DataInterface() {propertyValues {
        XCurve {keys [<at 0 1>]}
        YCurve {keys [<at 0 1.0067>]}
        ZCurve {keys [
          {interpMode=RCIM_Cubic, leaveTangent=285.2545, tangentMode=RCTM_Break, time=0, value=30}
          {interpMode=RCIM_Cubic, tangentMode=RCTM_Break, time=0.9989, value=100}
         ]}
       }})
     B=<0 1 1>
    )
   )
   &CameraOffset(Camera Offset Amount=20)
  ]
  rendererClasses [&NiagaraLightRendererProperties, &NiagaraSpriteRendererProperties]
  renderers [
   {rendererClass=&NiagaraLightRendererProperties, rendererIndex=0}
   {rendererClass=&NiagaraSpriteRendererProperties, rendererIndex=1}
  ]
 }
 FN_Emotes_ParticleDecal001 {
  emitterUpdateScript [
   &EmitterScalability(
    Module.bApplySpawnCountScale=false
    Module.bApplySpawnCountScaleByDistanceFraction=false
    Module.bUseMaxDistance=false
    Module.bUseMinDistance=false
   ) {enabled=false}
   &EmitterLifeCycle(MaxLoopCount=1, NextLoopDuration=0.1) {enabled=false}
   &EmitterState()
   &SpawnBurst_Instantaneous(Spawn Count=17, Use Spawn Probability=false)
   &SpawnRate(SpawnRate=15, Use Spawn Probability=false) {enabled=false}
  ]
  particleSpawnScript [
   &InitializeParticle(
    Color=<rgba 5 0.7 0.7 1.5>
    Lifetime=0.95
    Module.Write Color=true
    Module.Write Lifetime=true
    Module.Write Mass=true
    Module.Write Material Random=true
    Module.Write Position=true
    Module.Write Scale=false
    Module.Write SpriteRotation=true
    Module.Write SpriteSize=true
    Position=&SimulationPosition()
    Sprite Size=&Vector2DFromFloat(Value=75)
   )
   &CameraOffset(Camera Offset Amount=25)
   &SampleSkeletalMeshSkeleton(
    Mesh=DataInterface() {propertyValues {
      DefaultMesh=null
      ExcludeBoneName=root
      OptionalComponentClass=null
      PreviewMesh=/Game/Characters/Player/Male/Male_Avg_Base/Fortnite_M_Avg_Player
      SkinningMode=SkinOnTheFly
      SoftSourceActor=null
      SourceMode=Default
      UvSetIndex=0
      WholeMeshLOD=-1
      bExcludeBone=false
      bReadDeformedGeometry=true
      bRequireCurrentFrameData=true
      ComponentTags []
      FilteredBones [pelvis,Spine_02,upperarm_l,lowerarm_l,hand_l,upperarm_r,lowerarm_r,hand_r,neck_01,head,thigh_l,calf_l,thigh_r,calf_r,spine_05]
      FilteredSockets []
      MeshUserParameter {parameter {name=None, typeDefHandle {registeredTypeIndex=1}, varData []}}
      SamplingRegions []
     }}
    Specific Bone Index:NiagaraInt32=&UniformRangedInt(Minimum=0, Maximum=15, Randomness Mode=::NewEnumerator1)
   )
   /Niagara/Modules/Spawn/Location/SkeletalMeshSkeletonLocation(Module.Inherit Velocity=false, Module.Use Sampled Bone as Ribbon ID=true)
  ]
  particleUpdateScript [
   &ParticleState()
   &ScaleColor(
    Scale Alpha=&FloatFromCurve(FloatCurve=DataInterface() {propertyValues {Curve {keys [
         <at 0 0 RCIM_Cubic>
         {arriveTangent=-1.5091, interpMode=RCIM_Cubic, leaveTangent=-1.5091, tangentMode=RCTM_User, time=0.2, value=1}
         <at 1 0>
        ]}}})
    ScaleA=true
    ScaleRGB=true
   )
   &DynamicMaterialParameters(
    Index 0 Param 1=&FloatFromCurve(FloatCurve=DataInterface() {propertyValues {Curve {keys [
         {interpMode=RCIM_Cubic, leaveTangent=8.739, tangentMode=RCTM_Break, time=0, value=-0.5}
         {arriveTangent=10, interpMode=RCIM_Cubic, leaveTangent=10, time=0.07, value=0.27}
         {arriveTangent=8.2492, interpMode=RCIM_Cubic, leaveTangent=8.2492, tangentMode=RCTM_User, time=0.15, value=1}
         {arriveTangent=4.695, interpMode=RCIM_Cubic, tangentMode=RCTM_Break, time=1, value=5}
        ]}}})
    Index 0 Param 2=1
    Index 0 Param 3=&FloatFromCurve(FloatCurve=DataInterface() {propertyValues {Curve {keys [
         {interpMode=RCIM_Cubic, leaveTangent=2.9398, tangentMode=RCTM_Break, time=0, value=-0.25}
         {arriveTangent=0.3723, interpMode=RCIM_Cubic, tangentMode=RCTM_Break, time=1, value=1.4}
        ]}}})
    Index 0 Param 4=1
    Module.Param0WriteEnabled=true
    Module.Param1WriteEnabled=true
    Module.Param2WriteEnabled=true
    Module.Param3WriteEnabled=true
   )
   &SolveForcesAndVelocity(Clamp Velocity=false, Limit Acceleration=false)
  ]
  rendererClasses [&NiagaraSpriteRendererProperties]
  renderers [{rendererClass=&NiagaraSpriteRendererProperties}]
 }
 FN_Emotes_Starts001 {
  emitterUpdateScript [
   &EmitterScalability(
    Module.bApplySpawnCountScale=false
    Module.bApplySpawnCountScaleByDistanceFraction=false
    Module.bUseMaxDistance=false
    Module.bUseMinDistance=false
   ) {enabled=false}
   &EmitterLifeCycle(MaxLoopCount=1, NextLoopDuration=0.1) {enabled=false}
   &EmitterState()
   &SpawnBurst_Instantaneous(Spawn Count=&UniformRangedInt(Minimum=15, Maximum=10), Spawn Time=0.05, Use Spawn Probability=false)
   &SpawnRate(SpawnRate=3, Use Spawn Probability=false) {enabled=false}
  ]
  particleSpawnScript [
   &InitializeParticle(
    Color=&UniformRangedLinearColor(Minimum=<rgba 20 1 1 1>, Maximum=<rgba 20 2 2 1>)
    Lifetime=&UniformRangedFloat(Minimum=0.3, Maximum=0.8)
    Module.Write Color=true
    Module.Write Lifetime=true
    Module.Write Mass=true
    Module.Write Material Random=true
    Module.Write Position=true
    Module.Write Scale=false
    Module.Write SpriteRotation=true
    Module.Write SpriteSize=true
    Position=&SimulationPosition()
    Sprite Rotation=&UniformRangedFloat(Minimum=0.02, Maximum=-0.02)
    Sprite Size=&UniformRangedVector2D(Minimum=<3 20>, Maximum=<5 4.08>)
   )
   &CameraOffset(Camera Offset Amount=5)
   &SystemLocation(
    MaskSpawnGroup=false
    Offset=&UniformRangedVector(Minimum=<35 35 60>, Maximum=<-35 -35 -10>, Randomness Mode=::NewEnumerator1)
   )
   &AddVelocity(Velocity=&UniformRangedVector(Minimum=<20 20 200>, Maximum=<-20 -20 900>, Randomness Mode=::NewEnumerator1))
  ]
  particleUpdateScript [
   &ParticleState()
   &Drag(Drag=8, Use Linear Drag=true, Use Rotational Drag=false)
   &SolveForcesAndVelocity(Clamp Velocity=false, Limit Acceleration=false)
   &ScaleColor(
    Scale Alpha=&FloatFromCurve(FloatCurve=DataInterface() {propertyValues {Curve {keys [{arriveTangent=0.0145, interpMode=RCIM_Cubic, leaveTangent=0.0145, tangentMode=RCTM_User, time=0.4917, value=0.9535}, <at 1 0>]}}})
    Scale RGB=&VectorFromFloat(Value=&FloatFromCurve(FloatCurve=DataInterface() {propertyValues {Curve {keys [<at 0 200>,<at 0.0158 44.4879>,<at 0.1 1>]}}}))
    ScaleA=true
    ScaleRGB=true
   )
   &ScaleSpriteSize(Scale Factor=&Vector2DFromFloat(Value=&FloatFromCurve(
      FloatCurve=DataInterface() {propertyValues {Curve {keys [<at 0.0121 0.5565>,<at 0.491 0.8934>,<at 1.0011 0.002>]}}}
      Scale Curve=&UniformRangedFloat(Minimum=0.7, Maximum=1.5)
     )))
  ]
  rendererClasses [&NiagaraSpriteRendererProperties]
  renderers [{rendererClass=&NiagaraSpriteRendererProperties}]
 }
 Local_Screen_Glow_Cone001 {
  emitterUpdateScript [
   &EmitterScalability(
    Module.bApplySpawnCountScale=false
    Module.bApplySpawnCountScaleByDistanceFraction=false
    Module.bUseMaxDistance=false
    Module.bUseMinDistance=false
   ) {enabled=false}
   &EmitterLifeCycle(MaxLoopCount=1, NextLoopDuration=0.1) {enabled=false}
   &EmitterState()
   &SpawnBurst_Instantaneous(Spawn Count=2, Use Spawn Probability=false)
  ]
  particleSpawnScript [
   &InitializeParticle(
    Color=<rgba 1 1 1 0.3>
    Lifetime=&UniformRangedFloat(Minimum=0.3, Maximum=0.25)
    Module.Write Color=true
    Module.Write Lifetime=true
    Module.Write Mass=true
    Module.Write Material Random=true
    Module.Write Position=true
    Module.Write Scale=false
    Module.Write SpriteRotation=true
    Module.Write SpriteSize=true
    Position=&SimulationPosition()
    Sprite Rotation=&UniformRangedFloat(Minimum=0, Maximum=0)
    Sprite Size=&Vector2DFromFloat(Value=&UniformRangedFloat(Minimum=1, Maximum=20))
   )
   &AddVelocity(Velocity=<0 0 -0.01>)
   /SetVariables(Particles.LightPosition:Vector3f=/Niagara/DynamicInputs/Helpers/SimulationPosition(), Particles.Scale=<1 5 1>) {isSetParameters=true}
   /Niagara/Modules/Spawn/Orientation/V2/InitialMeshOrientation(Orientation Coordinate Space:ENiagaraCoordinateSpace=::Local, Rotation=<0.5 0 0>, Use Orientation Vector=true, Use Rotation Vector=true)
   &SystemLocation(MaskSpawnGroup=false)
  ]
  particleUpdateScript [
   &ParticleState()
   &SolveForcesAndVelocity(Clamp Velocity=false, Limit Acceleration=false)
   /SetVariables(Particles.Scale=&VectorFromCurve(VectorCurve=DataInterface() {propertyValues {
       XCurve {keys [
         {arriveTangent=0.3216, interpMode=RCIM_Cubic, leaveTangent=0.3216, tangentMode=RCTM_User, time=0, value=0}
         {interpMode=RCIM_Cubic, tangentMode=RCTM_User, time=1, value=0.2}
        ]}
       YCurve {keys [
         {arriveTangent=0.2996, interpMode=RCIM_Cubic, leaveTangent=0.2996, tangentMode=RCTM_User, time=0, value=0}
         {interpMode=RCIM_Cubic, tangentMode=RCTM_User, time=1.0004, value=1.5}
        ]}
       ZCurve {keys [
         {arriveTangent=0.9164, interpMode=RCIM_Cubic, leaveTangent=0.9164, tangentMode=RCTM_User, time=0, value=0}
         {interpMode=RCIM_Cubic, tangentMode=RCTM_User, time=1, value=1.5}
        ]}
      }})) {isSetParameters=true}
   /SetVariables(
    Particles.LightColor=&ScaleLinearColorByCurve(
     InitialValue=<rgba 1 1 1 1>
     ScaleCurve=DataInterface() {propertyValues {
       AlphaCurve {keys [
         {interpMode=RCIM_Cubic, tangentMode=RCTM_User, time=0.0012, value=1}
         {arriveTangent=-2.3333, interpMode=RCIM_Cubic, leaveTangent=-2.3333, tangentMode=RCTM_User, time=0.1765, value=0.601}
         {arriveTangent=-1, interpMode=RCIM_Cubic, leaveTangent=-1, tangentMode=RCTM_User, time=0.2889, value=0.3}
         {interpMode=RCIM_Cubic, tangentMode=RCTM_User, time=1, value=0}
        ]}
       BlueCurve {keys [
         {interpMode=RCIM_Cubic, tangentMode=RCTM_User, time=0, value=0}
         {arriveTangent=2, interpMode=RCIM_Cubic, leaveTangent=2, tangentMode=RCTM_User, time=0.1765, value=0.2858}
         {interpMode=RCIM_Cubic, tangentMode=RCTM_User, time=0.503, value=11.4694}
         {interpMode=RCIM_Cubic, tangentMode=RCTM_User, time=1, value=0}
        ]}
       GreenCurve {keys [
         {interpMode=RCIM_Cubic, tangentMode=RCTM_User, time=0, value=0}
         {arriveTangent=1, interpMode=RCIM_Cubic, leaveTangent=1, tangentMode=RCTM_User, time=0.1765, value=0.1429}
         {arriveTangent=1.1038, interpMode=RCIM_Cubic, leaveTangent=1.1038, tangentMode=RCTM_User, time=0.5, value=0.5}
         {interpMode=RCIM_Cubic, tangentMode=RCTM_User, time=0.5, value=0.5}
         {interpMode=RCIM_Cubic, tangentMode=RCTM_User, time=1, value=0}
        ]}
       RedCurve {keys [
         {interpMode=RCIM_Cubic, tangentMode=RCTM_User, time=0, value=0}
         {arriveTangent=30, interpMode=RCIM_Cubic, leaveTangent=30, tangentMode=RCTM_User, time=0.1781, value=1.2533}
         {interpMode=RCIM_Cubic, tangentMode=RCTM_User, time=0.487, value=2.5303}
         {interpMode=RCIM_Cubic, tangentMode=RCTM_User, time=1, value=0}
        ]}
      }}
     CurveIndex=$Particles.NormalizedAge
    )
    Particles.LightIntensity:Vector3f=&VectorFromFloat(Value=0)
    Particles.SpriteRotation:NiagaraFloat=&FloatFromCurve(FloatCurve=DataInterface() {propertyValues {Curve {keys [<at -0.0046 50>,<at 1.0011 0.002>]}}})
   ) {isSetParameters=true}
   &ScaleColor(
    Scale Alpha=&FloatFromCurve(FloatCurve=DataInterface() {propertyValues {Curve {keys [<at 0.2596 1.0113 RCIM_Cubic>,<at 1.0011 0.002>]}}})
    Scale RGB=&VectorFromCurve(VectorCurve=DataInterface() {propertyValues {
       XCurve {keys [{arriveTangent=-1.6489, interpMode=RCIM_Cubic, leaveTangent=-1.6489, tangentMode=RCTM_User, time=-0.0011, value=1.0067}, <at 0.9988 0>]}
       YCurve {keys [{arriveTangent=-1.8566, interpMode=RCIM_Cubic, leaveTangent=-1.8566, tangentMode=RCTM_User, time=-0.0011, value=1.0067}, <at 0.9988 0>]}
       ZCurve {keys [{arriveTangent=-1.9787, interpMode=RCIM_Cubic, leaveTangent=-1.9787, tangentMode=RCTM_User, time=-0.0011, value=1.0067}, <at 0.9988 0>]}
      }})
    ScaleA=true
    ScaleRGB=true
   )
  ]
  rendererClasses [
   &NiagaraMeshRendererProperties
   &NiagaraLightRendererProperties
   &NiagaraSpriteRendererProperties
  ]
  renderers [
   {rendererClass=&NiagaraMeshRendererProperties, rendererIndex=0}
   {rendererClass=&NiagaraLightRendererProperties, rendererIndex=1}
   {rendererClass=&NiagaraSpriteRendererProperties, rendererIndex=2}
  ]
 }
 Smoke001 {
  emitterUpdateScript [
   &EmitterState()
   &SpawnRate(SpawnRate=50, Use Spawn Probability=false) {enabled=false}
   &SpawnBurst_Instantaneous(Spawn Count=8, Spawn Time=0.1, Use Spawn Probability=false)
  ]
  particleSpawnScript [
   &InitializeParticle(
    Lifetime=&UniformRangedFloat(Minimum=0.8, Maximum=1.3)
    Module.Write Color=true
    Module.Write Lifetime=true
    Module.Write Mass=true
    Module.Write Material Random=true
    Module.Write Position=false
    Module.Write Scale=false
    Module.Write SpriteRotation=true
    Module.Write SpriteSize=true
    Position=/Niagara/Functions/Localspace/SimulationPosition()
    Sprite Rotation=&UniformRangedFloat(Minimum=0, Maximum=180)
    Sprite Size=&UniformRangedVector2D(Minimum=<200 200>, Maximum=<150 150>)
   )
   &AddVelocity(Velocity=<0 0 15>)
   &CylinderLocation(
    Cylinder Height=60
    Cylinder Radius=30
    Hemisphere X=true
    Lathe Profile=DataInterface() {propertyValues {Curve {keys [<at 0 0>,<at 0.9 0>,<at 1.0013 1>]}}}
    MaskSpawnGroup=false
    Module.Enable Lathe Profile=false
    Module.Override Randomness=false
    Module.Surface Only=true
    Offset=<-20 0 60>
    Surface Only Band Thickness=0
    Use Endcaps In Surface Only Mode=false
   )
  ]
  particleUpdateScript [
   &ParticleState()
   /Niagara/Modules/Update/Forces/CurlNoiseForce(Noise Frequency=0.2, Noise Strength=25, Pan Noise=false)
   &Drag(Drag=3, Use Linear Drag=true, Use Rotational Drag=false)
   &SolveForcesAndVelocity(Clamp Velocity=false, Limit Acceleration=false)
   /Niagara/Modules/Update/SubUV/SubUVAnimation(
    Number Of Frames=64
    SubUV Lookup Index=&FloatFromCurve(FloatCurve=DataInterface() {propertyValues {Curve {keys [
         {arriveTangent=0.7985, interpMode=RCIM_Cubic, leaveTangent=0.7985, tangentMode=RCTM_User, time=0, value=0.6}
         {arriveTangent=0.0289, interpMode=RCIM_Cubic, leaveTangent=0.0289, tangentMode=RCTM_User, time=1.0011, value=0.985}
        ]}}})
   )
   &ScaleSpriteSize(Scale Factor=&Vector2DFromFloat(Value=&FloatFromCurve(FloatCurve=DataInterface() {propertyValues {Curve {keys [
          <at -0.0046 0.25 RCIM_Cubic>
          {arriveTangent=-0.1882, interpMode=RCIM_Cubic, leaveTangent=-0.1882, tangentMode=RCTM_User, time=1.0099, value=0.8547}
         ]}}})))
   &Color(Color=&ColorFromCurve(ColorCurve=DataInterface() {propertyValues {
       AlphaCurve {keys [<at -0.0382 0>,<at 0.004 1 RCIM_Cubic>,<at 0.9977 0 RCIM_Cubic>]}
       BlueCurve {keys [<at 0.0273 0.0156>,<at 0.1487 0.006>]}
       GreenCurve {keys [<at 0.0273 0.0484>,<at 0.1487 0.0056>]}
       RedCurve {keys [<at 0.0273 1>,<at 0.1487 0.026>]}
      }}))
  ]
  rendererClasses [&NiagaraSpriteRendererProperties]
  renderers [{rendererClass=&NiagaraSpriteRendererProperties}]
 }
}
summary {
 systemName=NS_Emote_NeonCatSpy_Red
 emitters [
  {emitterName=FN_Emotes_ParticleDecal001, rendererClasses [&NiagaraSpriteRendererProperties]}
  {
   emitterName=FN_Emotes_Glow001
   rendererClasses [&NiagaraLightRendererProperties, &NiagaraSpriteRendererProperties]
  }
  {emitterName=FN_Emotes_Starts001, rendererClasses [&NiagaraSpriteRendererProperties]}
  {
   emitterName=Local_Screen_Glow_Cone001
   rendererClasses [
    &NiagaraMeshRendererProperties
    &NiagaraLightRendererProperties
    &NiagaraSpriteRendererProperties
   ]
  }
  {emitterName=Smoke001, rendererClasses [&NiagaraSpriteRendererProperties]}
  {emitterName=FLare001, rendererClasses [&NiagaraSpriteRendererProperties, &NiagaraLightRendererProperties]}
 ]
 userVariables [User.InitialOffset {type=Vector3f, description="", defaultValue {struct=Vector3f, value=<0 0 0>}}]
}
